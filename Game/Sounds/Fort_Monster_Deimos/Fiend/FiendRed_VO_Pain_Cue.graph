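# Sound cue for the red Fiend's pain vocalizations, a Deimos monster creature voice.
# A mixer blends a close layer with a quieter distant layer, each picking randomly from six takes through a modulator and its own close or distant attenuation.
# The whole cue is pitched down, routed to the creature vocalization sound class, and limited by global and local Deimos vocal concurrency with virtualization disabled.
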SoundCue /Game/Sounds/Fort_Monster_Deimos/Fiend/FiendRed_VO_Pain_Cue
root=Mixer_0
nodes [
 {id=Attenuation_0, children [Modulator_0], properties {attenuationSettings=/Game/Sounds/Attenuation/Fort_Deimos_Vocal_Close_Att}}
 {id=Attenuation_1, children [Modulator_1], properties {attenuationSettings=/Game/Sounds/Attenuation/Fort_Deimos_Vocal_Distant_Att}}
 {id=Mixer_0, children [Attenuation_0, Attenuation_1], properties {inputVolume [1.1,0.5]}}
 {id=Modulator_0, children [Random_1], properties {pitchMax=1.2}}
 {id=Modulator_1, children [Random_0]}
 {
  id=Random_0
  children [WavePlayer_0, WavePlayer_1, WavePlayer_2, WavePlayer_3, WavePlayer_4, WavePlayer_11]
  properties {bShouldExcludeFromBranchCulling=true, weights [1,1,1,1,1,1]}
 }
 {
  id=Random_1
  children [WavePlayer_5, WavePlayer_6, WavePlayer_7, WavePlayer_8, WavePlayer_9, WavePlayer_10]
  properties {bShouldExcludeFromBranchCulling=true, weights [1,1,1,1,1,1]}
 }
 {id=WavePlayer_0, properties {soundWaveAssetPtr=&/Fiend_Pain_New_Distant_01}}
 {id=WavePlayer_1, properties {soundWaveAssetPtr=&/Fiend_Pain_New_Distant_02}}
 {id=WavePlayer_10, properties {soundWaveAssetPtr=&/Fiend_Pain_New_06}}
 {id=WavePlayer_11, properties {soundWaveAssetPtr=&/Fiend_Pain_New_Distant_07}}
 {id=WavePlayer_2, properties {soundWaveAssetPtr=&/Fiend_Pain_New_Distant_03}}
 {id=WavePlayer_3, properties {soundWaveAssetPtr=&/Fiend_Pain_New_Distant_04}}
 {id=WavePlayer_4, properties {soundWaveAssetPtr=&/Fiend_Pain_New_Distant_06}}
 {id=WavePlayer_5, properties {soundWaveAssetPtr=&/Fiend_Pain_New_02}}
 {id=WavePlayer_6, properties {soundWaveAssetPtr=&/Fiend_Pain_New_07}}
 {id=WavePlayer_7, properties {soundWaveAssetPtr=&/Fiend_Pain_New_01}}
 {id=WavePlayer_8, properties {soundWaveAssetPtr=&/Fiend_Pain_New_03}}
 {id=WavePlayer_9, properties {soundWaveAssetPtr=&/Fiend_Pain_New_04}}
]
properties {
 bBypassVolumeScaleForPriority=true
 maxDistance=24000
 pitchMultiplier=0.6
 priority=0.75
 soundClassObject=&Fort_Creatures_Vocalization_Tactical
 virtualizationMode=Disabled
 volumeMultiplier=1.7955
 attenuationOverrides {distanceAlgorithm=NaturalSound, falloffDistance=4000, attenuationShapeExtents {x=1000}}
 concurrencyOverrides {resolutionRule=PreventNew}
 concurrencySet [/Game/Sounds/Fort_Monster_Deimos/Settings/Deimos_Vocal_Global_Conc, /Game/Sounds/Fort_Monster_Deimos/Settings/Deimos_Vocal_Local_Conc]
}
