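# Sound cue for the red Fiend monster's walking breath vocalization, drawing on the Fiend run-breath recordings.
# A short random delay feeds a modulator, which varies one of twelve equally weighted breath waves picked at random.
# It plays as a creature vocalization with close vocal attenuation, using Deimos global and local vocal concurrency sets and disabled virtualization.
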SoundCue /Game/Sounds/Fort_Monster_Deimos/Fiend/FiendRed_VO_Walk_Breath_Cue
root=Delay_0
nodes [
 {id=Delay_0, children [Modulator_1], properties {delayMax=1.25}}
 {id=Modulator_1, children [Random_3]}
 {
  id=Random_3
  children [
   WavePlayer_5
   WavePlayer_6
   WavePlayer_7
   WavePlayer_8
   WavePlayer_18
   WavePlayer_19
   WavePlayer_20
   WavePlayer_21
   WavePlayer_22
   WavePlayer_23
   WavePlayer_24
   WavePlayer_25
  ]
  properties {bShouldExcludeFromBranchCulling=true, weights [1,1,1,1,1,1,1,1,1,1,1,1]}
 }
 {id=WavePlayer_18, properties {soundWaveAssetPtr=&/Fiend_RunBreathNew_02}}
 {id=WavePlayer_19, properties {soundWaveAssetPtr=&/Fiend_RunBreathNew_03}}
 {id=WavePlayer_20, properties {soundWaveAssetPtr=&/Fiend_RunBreathNew_04}}
 {id=WavePlayer_21, properties {soundWaveAssetPtr=&/Fiend_RunBreathNew_05}}
 {id=WavePlayer_22, properties {soundWaveAssetPtr=&/Fiend_RunBreathNew_06}}
 {id=WavePlayer_23, properties {soundWaveAssetPtr=&/Fiend_RunBreathNew_07}}
 {id=WavePlayer_24, properties {soundWaveAssetPtr=&/Fiend_RunBreathNew_08}}
 {id=WavePlayer_25, properties {soundWaveAssetPtr=&/Fiend_RunBreathNew_09}}
 {id=WavePlayer_5, properties {soundWaveAssetPtr=&/Fiend_RunBreathNew_10}}
 {id=WavePlayer_6, properties {soundWaveAssetPtr=&/Fiend_RunBreathNew_11}}
 {id=WavePlayer_7, properties {soundWaveAssetPtr=&/Fiend_RunBreathNew_12}}
 {id=WavePlayer_8, properties {soundWaveAssetPtr=&/Fiend_RunBreathNew_01}}
]
properties {
 attenuationSettings=/Game/Sounds/Attenuation/Fort_Deimos_Vocal_Close_Att
 bBypassVolumeScaleForPriority=true
 maxDistance=13000
 priority=0.75
 soundClassObject=/Game/Sounds/FortSoundClassesAndModes/Fort_Creatures_Vocalization_Normal
 virtualizationMode=Disabled
 volumeMultiplier=0.998
 attenuationOverrides {distanceAlgorithm=NaturalSound, falloffDistance=4000, attenuationShapeExtents {x=1000}}
 concurrencyOverrides {maxCount=2, volumeScale=0.75, volumeScaleAttackTime=0.1}
 concurrencySet [/Game/Sounds/Fort_Monster_Deimos/Settings/Deimos_Vocal_Global_Conc, /Game/Sounds/Fort_Monster_Deimos/Settings/Deimos_Vocal_Local_Conc]
}
