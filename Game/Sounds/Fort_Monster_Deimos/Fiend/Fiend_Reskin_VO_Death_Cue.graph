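# Sound cue for the death vocalization of the reskinned Deimos Fiend monster.
# A random node picks one of four equally weighted death wave files, and a modulator raises the upper limit of its pitch variation.
# It plays boosted in the tactical creature vocalization class, using close vocal attenuation and global and local Deimos vocal concurrency sets.
# Virtualization is disabled, and volume scaling does not affect its priority.
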SoundCue /Game/Sounds/Fort_Monster_Deimos/Fiend/Fiend_Reskin_VO_Death_Cue
root=Modulator_0
nodes [
 {id=Modulator_0, children [Random_2], properties {pitchMax=1.2}}
 {
  id=Random_2
  children [WavePlayer_13, WavePlayer_14, WavePlayer_15, WavePlayer_16]
  properties {bShouldExcludeFromBranchCulling=true, weights [1,1,1,1]}
 }
 {id=WavePlayer_13, properties {soundWaveAssetPtr=&/DEIMOS_Fiend_Death_010}}
 {id=WavePlayer_14, properties {soundWaveAssetPtr=&/DEIMOS_Fiend_Death_011}}
 {id=WavePlayer_15, properties {soundWaveAssetPtr=&/DEIMOS_Fiend_Death_012}}
 {id=WavePlayer_16, properties {soundWaveAssetPtr=&/DEIMOS_Fiend_Death_013}}
]
properties {
 attenuationSettings=/Game/Sounds/Attenuation/Fort_Deimos_Vocal_Close_Att
 bBypassVolumeScaleForPriority=true
 maxDistance=13000
 priority=0.75
 soundClassObject=&Fort_Creatures_Vocalization_Tactical
 virtualizationMode=Disabled
 volumeMultiplier=2.6937
 concurrencyOverrides {resolutionRule=PreventNew}
 concurrencySet [/Game/Sounds/Fort_Monster_Deimos/Settings/Deimos_Vocal_Global_Conc, /Game/Sounds/Fort_Monster_Deimos/Settings/Deimos_Vocal_Local_Conc]
}
