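# Sound cue for the Fiend monster's attack vocalizations in the Deimos creature set.
# A Random node picks one of seven equally weighted attack wave variants, excluded from branch culling, and feeds it through a Modulator for per-play variation.
# It plays as a tactical creature vocalization at high priority with close vocal attenuation, virtualization disabled, and global and local Deimos vocal concurrency sets limiting overlap.
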SoundCue /Game/Sounds/Fort_Monster_Deimos/Fiend/Fiend_VO_Attack_Cue
root=Modulator_0
nodes [
 {id=Modulator_0, children [Random_1]}
 {
  id=Random_1
  children [WavePlayer_7, WavePlayer_8, WavePlayer_9, WavePlayer_10, WavePlayer_11, WavePlayer_12, WavePlayer_13]
  properties {bShouldExcludeFromBranchCulling=true, weights [1,1,1,1,1,1,1]}
 }
 {id=WavePlayer_10, properties {soundWaveAssetPtr=&/Fiend_Attack_New_04}}
 {id=WavePlayer_11, properties {soundWaveAssetPtr=&/Fiend_Attack_New_05}}
 {id=WavePlayer_12, properties {soundWaveAssetPtr=&/Fiend_Attack_New_06}}
 {id=WavePlayer_13, properties {soundWaveAssetPtr=&/Fiend_Attack_New_07}}
 {id=WavePlayer_7, properties {soundWaveAssetPtr=&/Fiend_Attack_New_01}}
 {id=WavePlayer_8, properties {soundWaveAssetPtr=&/Fiend_Attack_New_02}}
 {id=WavePlayer_9, properties {soundWaveAssetPtr=&/Fiend_Attack_New_03}}
]
properties {
 attenuationSettings=/Game/Sounds/Attenuation/Fort_Deimos_Vocal_Close_Att
 bBypassVolumeScaleForPriority=true
 maxDistance=13000
 priority=0.9
 soundClassObject=&Fort_Creatures_Vocalization_Tactical
 virtualizationMode=Disabled
 volumeMultiplier=2.394
 attenuationOverrides {distanceAlgorithm=NaturalSound, falloffDistance=4200, attenuationShapeExtents {x=800}}
 concurrencyOverrides {resolutionRule=PreventNew}
 concurrencySet [/Game/Sounds/Fort_Monster_Deimos/Settings/Deimos_Vocal_Global_Conc, /Game/Sounds/Fort_Monster_Deimos/Settings/Deimos_Vocal_Local_Conc]
}
