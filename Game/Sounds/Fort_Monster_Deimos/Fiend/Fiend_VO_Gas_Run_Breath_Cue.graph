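# Sound cue for the running breath vocalization of the gas Fiend, a Deimos monster.
# A random node picks one of four equally weighted breath waves and passes it through a default Modulator for slight pitch and volume variation.
# Playback uses the creature vocalization sound class, close vocal attenuation, and both global and local Deimos vocal concurrency groups, with virtualization disabled.
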SoundCue /Game/Sounds/Fort_Monster_Deimos/Fiend/Fiend_VO_Gas_Run_Breath_Cue
root=Modulator_1
nodes [
 {id=Modulator_1, children [Random_0]}
 {
  id=Random_0
  children [WavePlayer_0,WavePlayer_1,WavePlayer_2,WavePlayer_3]
  properties {bShouldExcludeFromBranchCulling=true, weights [1,1,1,1]}
 }
 {id=WavePlayer_0, properties {soundWaveAssetPtr=&/Fiend_Reskin_Gas_RunBreath_04}}
 {id=WavePlayer_1, properties {soundWaveAssetPtr=&/Fiend_Reskin_Gas_RunBreath_03}}
 {id=WavePlayer_2, properties {soundWaveAssetPtr=&/Fiend_Reskin_Gas_RunBreath_02}}
 {id=WavePlayer_3, properties {soundWaveAssetPtr=&/Fiend_Reskin_Gas_RunBreath_01}}
]
properties {
 attenuationSettings=/Game/Sounds/Attenuation/Fort_Deimos_Vocal_Close_Att
 bBypassVolumeScaleForPriority=true
 maxDistance=13000
 pitchMultiplier=0.98
 priority=0.75
 soundClassObject=/Game/Sounds/FortSoundClassesAndModes/Fort_Creatures_Vocalization_Normal
 virtualizationMode=Disabled
 volumeMultiplier=1.596
 attenuationOverrides {distanceAlgorithm=NaturalSound, falloffDistance=4000, attenuationShapeExtents {x=1000}}
 concurrencyOverrides {maxCount=2, volumeScale=0.75, volumeScaleAttackTime=0.1}
 concurrencySet [/Game/Sounds/Fort_Monster_Deimos/Settings/Deimos_Vocal_Global_Conc, /Game/Sounds/Fort_Monster_Deimos/Settings/Deimos_Vocal_Local_Conc]
}
