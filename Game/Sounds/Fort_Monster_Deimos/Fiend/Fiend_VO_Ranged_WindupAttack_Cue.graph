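# Sound cue for the Deimos Fiend monster's vocalization while winding up a ranged attack.
# A Random node picks one of four equally weighted windup wave variations and feeds a Modulator at the root.
# It plays as a tactical creature vocalization with boosted volume, high priority, close-vocal attenuation and no virtualization.
# Shared global and local Deimos vocal concurrency sets govern simultaneous playback.
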SoundCue /Game/Sounds/Fort_Monster_Deimos/Fiend/Fiend_VO_Ranged_WindupAttack_Cue
root=Modulator_0
nodes [
 {id=Modulator_0, children [Random_1]}
 {
  id=Random_1
  children [WavePlayer_3,WavePlayer_4,WavePlayer_5,WavePlayer_6]
  properties {bShouldExcludeFromBranchCulling=true, weights [1,1,1,1]}
 }
 {id=WavePlayer_3, properties {soundWaveAssetPtr=&/DEIMOS_Fiend_WindupAttack_010}}
 {id=WavePlayer_4, properties {soundWaveAssetPtr=&/DEIMOS_Fiend_WindupAttack_011}}
 {id=WavePlayer_5, properties {soundWaveAssetPtr=&/DEIMOS_Fiend_WindupAttack_012}}
 {id=WavePlayer_6, properties {soundWaveAssetPtr=&/DEIMOS_Fiend_WindupAttack_013}}
]
properties {
 attenuationSettings=/Game/Sounds/Attenuation/Fort_Deimos_Vocal_Close_Att
 bBypassVolumeScaleForPriority=true
 maxDistance=13000
 priority=0.9
 soundClassObject=&Fort_Creatures_Vocalization_Tactical
 virtualizationMode=Disabled
 volumeMultiplier=2.9925
 attenuationOverrides {distanceAlgorithm=NaturalSound, falloffDistance=4200, attenuationShapeExtents {x=800}}
 concurrencyOverrides {resolutionRule=PreventNew}
 concurrencySet [/Game/Sounds/Fort_Monster_Deimos/Settings/Deimos_Vocal_Global_Conc, /Game/Sounds/Fort_Monster_Deimos/Settings/Deimos_Vocal_Local_Conc]
}
